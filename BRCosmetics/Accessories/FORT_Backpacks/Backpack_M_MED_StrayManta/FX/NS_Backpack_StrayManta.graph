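# Niagara idle glow effect for the StrayManta backpack cosmetic, in warm amber-gold.
# Its single Idle emitter burst-spawns particles that outlive their lifetime, each drawn by sprite, light and mesh renderers.
# The User.bisFrontEnd flag picks the sprite color, with lower alpha in the front end.
# FN_Light_Attributes writes the light attributes, placing it below the particle, while the sprite gets a small camera offset.
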NiagaraSystem /BRCosmetics/Accessories/FORT_Backpacks/Backpack_M_MED_StrayManta/FX/NS_Backpack_StrayManta
dependencies {usedDataInterfaces [], usedDynamicInputs [], usedModules [&SystemState]}
emitters {Idle {
  emitterUpdateScript [
   &EmitterState(Life Cycle Mode=::NewEnumerator1, Loop Behavior=::NewEnumerator1, Loop Duration Mode=::NewEnumerator1, UseLoopDelay=false)
   &SpawnBurst_Instantaneous(Use Spawn Probability=false, UseLoopCountLimit=false)
  ]
  particleSpawnScript [
   &InitializeParticle(
    Color=<rgba 1 0.7007 0.0187 0.16>
    Mesh Scale=<0.4 0.4 -0.45>
    Mesh Scale Mode=::NewEnumerator4
    Sprite Size Mode=::NewEnumerator3
    Uniform Sprite Size=12
    UsePositionOffset=false
    Write Material Random=true
   )
   /SetVariables(Particles.SpriteColor=&MakeCustomLinearColorFromBool(Bool=$User.bisFrontEnd, True Color=<rgba 1 0.5831 0.0126 0.25>, False Color=<rgba 1 0.5574 0 0.5>)) {isSetParameters=true}
  ]
  particleUpdateScript [
   &ParticleState(Kill Particles When Lifetime Has Elapsed=false)
   &FN_Light_Attributes(
    Color=<rgba 1 0.6423 0.0251 1>
    Create Color Attribute=true
    Create Exponent Attribute=true
    Create Position Attribute=true
    Create Radius Attribute=true
    Exponent=3
    Position=&AddVectorToPosition(Position=$Particles.Position, Vector=<0 0 -14>)
    Radius=22
    Scale Color=8
   )
   /SetVariables(Particles.SpriteCameraOffset=2) {isSetParameters=true}
  ]
  rendererClasses [
   &NiagaraSpriteRendererProperties
   &NiagaraLightRendererProperties
   &NiagaraMeshRendererProperties
  ]
  renderers [
   {rendererClass=&NiagaraSpriteRendererProperties}
   {rendererClass=&NiagaraLightRendererProperties, rendererIndex=1}
   {rendererClass=&NiagaraMeshRendererProperties, rendererIndex=2}
  ]
 }}
summary {
 systemName=NS_Backpack_StrayManta
 emitters [{
   emitterName=Idle
   rendererClasses [
    &NiagaraSpriteRendererProperties
    &NiagaraLightRendererProperties
    &NiagaraMeshRendererProperties
   ]
  }]
 userVariables [User.bisFrontEnd {type=NiagaraBool}]
}
